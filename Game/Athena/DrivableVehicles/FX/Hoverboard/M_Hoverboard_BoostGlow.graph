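# Unlit, additive, two-sided mesh-particle material for the hoverboard boost glow.
# Panning star and chromatic smoke textures, shaped by V-gradient powers, are tinted by DarkColor and particle color for emissive.
# Opacity combines particle alpha, the gradient and depth fade; mobile drops depth fade and the star layer.
# Dynamic parameter R and B drive a vertex offset weighted by inverted-V falloffs, presumably world position offset as the parameter names indicate.
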Material /Game/Athena/DrivableVehicles/FX/Hoverboard/M_Hoverboard_BoostGlow
graph {
 properties {
  bUsedWithMeshParticles=true
  bUsedWithNiagaraMeshParticles=true
  blendMode=BLEND_Additive
  pixelDepthOffsetMode=PDOM_Legacy
  previewMesh=&/SM_Hoverboard_BoostGlow
  shadingModel=MSM_Unlit
  twoSided=true
  usedShadingModels=Unlit
 }
 nodes [
  n1 = TextureSample(UVs=Panner(Coordinate=TextureCoordinate() {uTiling=2, vTiling=0.05}) {speedX=-0.05, speedY=2}) {samplerType=SAMPLERTYPE_Masks, texture=/Game/Characters/Player/Male/Medium/Bodies/M_Med_Soldier_04/Skins/Galaxy/Textures/Celestial_T_Stars}
  n2 = Multiply(A=@n1.B) {constB=16}
  n3 = ComponentMask(None=TextureCoordinate()) {g=true}
  n4 = Power(Base=@n3) {constExponent=16}
  n5 = TextureCoordinate() {uTiling=2, vTiling=0.1}
  n6 = LinearInterpolate(A=TextureSample(UVs=Panner(Coordinate=@n5) {speedX=0.05, speedY=2}) {texture=&T_Inky_Smoke_Tile_Chromatic}.RGB) {constAlpha=0.15}
  n7 = Multiply(A=Add(A=@n2, B=Multiply(A=@n2, B=@n4)), B=@n6)
  n8 = Constant()
  n9 = Power(Base=@n3)
  n10 = Multiply(
   A=Add(A=ShadingPathSwitch(Default=@n7, Mobile=@n8) {desc=Migrated from FeatureLevelSwitch}, B=@n6)
   B=LinearInterpolate(B=VectorParameter() {parameterName=DarkColor, defaultValue=<rgba 0.2108 0.2068 0.6042 0>}.RGB, Alpha=@n9) {constA=1}
  )
  n11 = ParticleColor()
  MaterialFunctionCall(Emissive=Multiply(A=Add(A=@n10, B=@n4), B=@n11.RGB)) {materialFunction=&MF_FX_EmissiveAdjustments}
  n12 = LinearInterpolate(Alpha=@n9) {constA=20, constB=0}
  Multiply(
   A=Multiply(A=@n11.A, B=@n9)
   B=ShadingPathSwitch(Default=DepthFade(FadeDistance=@n12), Mobile=Constant() {r=1}) {desc=Migrated from FeatureLevelSwitch}
  )
  n13 = DynamicParameter() {paramNames [R,G,B,Param4]}
  n14 = OneMinus(None=ComponentMask(None=TextureCoordinate()) {g=true})
  n15 = Multiply(A=@n14) {constB=0.1}
  n16 = Multiply(A=Power(Base=@n14, Exp=ScalarParameter() {defaultValue=4, parameterName=WPOMaskPower}) {constExponent=56}) {constB=0.25}
  n17 = Power(Base=@n14, Exp=ScalarParameter() {defaultValue=16, parameterName=WPOMaskPowerOuter})
  Multiply(A=AppendVector(A=AppendVector(A=@n13.R, B=@n13.R), B=@n13.B), B=Add(A=Add(A=@n15, B=@n16), B=Multiply(A=@n17) {constB=0.65}))
 ]
}
