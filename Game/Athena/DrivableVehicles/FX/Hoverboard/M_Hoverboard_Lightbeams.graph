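# Translucent, two-sided unlit material for hoverboard light beam mesh particles.
# A Fresnel, vertex-red and UV gradient mask blends dim to bright color, with a hot end and panning smoke noise scaled by the "Panning" dynamic parameter.
# Opacity combines that mask with particle alpha, optional noise and a sine pulse.
# View-angle, depth and near-camera fades apply only on the deferred shading path.
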Material /Game/Athena/DrivableVehicles/FX/Hoverboard/M_Hoverboard_Lightbeams
graph {
 properties {
  bUsedWithMeshParticles=true
  bUsedWithNiagaraMeshParticles=true
  blendMode=BLEND_Translucent
  pixelDepthOffsetMode=PDOM_Legacy
  shadingModel=MSM_Unlit
  twoSided=true
  usedShadingModels=Unlit
 }
 nodes [
  n1 = ParticleColor()
  n2 = TextureCoordinate()
  n3 = VectorParameter() {parameterName=NoiseTile, defaultValue=<rgba 2 0.2 0 0>}
  n4 = AppendVector(A=@n3.R, B=@n3.R)
  n5 = Time()
  n6 = ScalarParameter() {defaultValue=1, parameterName=PanSpeed}
  n7 = TextureSample(UVs=Panner(Coordinate=Multiply(A=@n2, B=@n4), Time=Multiply(A=@n5, B=@n6)) {speedX=0.25, speedY=2.5}) {texture=&T_Inky_Smoke_Tile_Chromatic}
  n8 = Fresnel(BaseReflectFractionIn=ScalarParameter() {defaultValue=0.2, parameterName=FresnelBaseBrightness}) {baseReflectFraction=0.2, exponent=1.5}
  n9 = ComponentMask(None=TextureCoordinate()) {g=true}
  n10 = Power(Base=@n9, Exp=ScalarParameter() {defaultValue=3, parameterName=GradientPower}) {constExponent=3}
  n11 = Multiply(A=Power(Base=VertexColor().R, Exp=ScalarParameter() {defaultValue=1, parameterName=VertColorPower}), B=@n10)
  n12 = Add(
   A=LinearInterpolate(
    A=VectorParameter() {parameterName=DimColor, defaultValue=<rgba 0.0729 0.3476 1 0>}.RGB
    B=VectorParameter() {parameterName=BrightColor, defaultValue=<rgba 0.1156 1.8207 2 0>}.RGB
    Alpha=Multiply(A=@n8, B=@n11)
   )
   B=Multiply(A=Power(Base=@n9) {constExponent=12})
  )
  n13 = Multiply(A=Multiply(A=@n7.RGB, B=@n12), B=DynamicParameter() {defaultValue=<rgba 0.5 1 1 1>, paramNames [Panning,Param2,Param3,Param4]}.Panning)
  MaterialFunctionCall(Emissive=Multiply(A=@n1.RGB, B=Add(A=@n13, B=@n12))) {materialFunction=&MF_FX_EmissiveAdjustments}
  n14 = ScalarParameter() {parameterName=NoiseOpacityInfluence}
  n15 = Constant() {r=1}
  n16 = DotProduct(A=CameraVectorWS(), B=VertexNormalWS())
  n17 = Multiply(A=Power(Base=Abs(None=@n16), Exp=ScalarParameter() {defaultValue=2, parameterName=AngleFadePower})) {constB=4}
  n18 = ScalarParameter() {defaultValue=40, parameterName=DepthFade}
  n19 = Multiply(
   A=DepthFade(Opacity=Clamp(None=@n17), FadeDistance=LinearInterpolate(A=@n18, Alpha=@n10) {constA=40, constB=0})
   B=MaterialFunctionCall(Distance=ScalarParameter() {defaultValue=300, parameterName=NearFadeDistance}) {materialFunction=&Near_Camera_Fade}.Result
  )
  n20 = Multiply(
   A=Multiply(A=Multiply(A=@n8, B=@n11), B=ShadingPathSwitch(Default=@n15, Deferred=@n19))
   B=ScalarParameter() {defaultValue=0.8, parameterName=Opacity}
  )
  n21 = LinearInterpolate(
   A=ScalarParameter() {defaultValue=0.4, parameterName=MinPulseBrightness}
   Alpha=MaterialFunctionCall(Speed=Constant() {r=0.5}) {materialFunction=&DebugTimeSine}.Result
  ) {constA=0.4}
  Multiply(A=@n1.A, B=Multiply(A=LinearInterpolate(B=@n7.RGB, Alpha=@n14) {constA=1}, B=Multiply(A=@n20, B=@n21)))
 ]
}
