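# Translucent unlit particle material for the hoverboard shimmer effect, producing tinted glowing rings with refractive distortion.
# Panning smoke noise in radial and flat UVs, blended by the Blur dynamic parameter, tints two scrolling corona-gradient layers mapped through sphere masks.
# The noise, shaped by a sine-remapped sphere mask, particle alpha and a near/far depth-fade window, drives refraction strength.
# The Random dynamic parameter offsets panning time per particle.
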Material /Game/Athena/DrivableVehicles/FX/Hoverboard/M_Hoverboard_Shimmer
graph {
 properties {
  bUsedWithNiagaraSprites=true
  bUsedWithParticleSprites=true
  blendMode=BLEND_Translucent
  pixelDepthOffsetMode=PDOM_Legacy
  refractionMethod=RM_IndexOfRefraction
  shadingModel=MSM_Unlit
  usedShadingModels=Unlit
 }
 nodes [
  n1 = ParticleColor()
  n2 = Constant() {r=1}
  n3 = ScalarParameter() {defaultValue=0.5, parameterName=NoiseYTile}
  n4 = Multiply(
   A=MaterialFunctionCall() {materialFunction=&VectorToRadialValue}.Radial Coordinates
   B=AppendVector(A=@n2, B=@n3)
  )
  n5 = DynamicParameter() {defaultValue=<rgba 1 0 1 1>, paramNames [Random,Blur,Param3,Param4]}
  n6 = Add(A=@n5.Random, B=Time())
  n7 = Multiply(A=ScalarParameter() {defaultValue=1, parameterName=NoisePanSpeed}, B=@n6)
  n8 = Multiply(A=AppendVector(A=@n3, B=@n2), B=TextureCoordinate() {uTiling=0.25, vTiling=0.8})
  n9 = LinearInterpolate(
   A=TextureSample(UVs=Panner(Coordinate=@n4, Time=@n7) {speedX=0.25, speedY=-2}) {texture=&T_Inky_Smoke_Tile_Chromatic}.RGB
   B=TextureSample(UVs=Panner(Coordinate=@n8, Time=@n7) {speedX=-1, speedY=0.1}) {texture=&T_Inky_Smoke_Tile_Chromatic}.RGB
   Alpha=@n5.Blur
  ) {constAlpha=1}
  n10 = TextureCoordinate()
  n11 = Constant() {r=0.5}
  n12 = SphereMask(A=@n10, B=@n11, Radius=@n11) {attenuationRadius=0.5, hardnessPercent=0}
  n13 = ConstantBiasScale(None=Sine(None=@n12) {period=1.3})
  n14 = Multiply(A=@n1.A, B=Multiply(A=@n9, B=Multiply(A=@n13, B=@n12)))
  n15 = ScalarParameter() {defaultValue=100, parameterName=MaxDepthVis}
  n16 = Multiply(A=OneMinus(None=DepthFade(FadeDistance=@n15)), B=DepthFade(FadeDistance=ScalarParameter() {defaultValue=8, parameterName=MinDepthVis}))
  LinearInterpolate(B=ScalarParameter() {defaultValue=1.1, parameterName=Refraction}, Alpha=ComponentMask(None=Multiply(A=@n14, B=@n16)) {g=true}) {constA=1}
  n17 = Add(A=@n9, B=VectorParameter() {parameterName=ColorAdd, defaultValue=<rgba 0.9583 1.8519 2 0>}.RGB) {constB=0.5}
  n18 = ComponentMask(None=@n10) {g=true}
  n19 = Power(Base=LinearInterpolate(A=@n12, B=SphereMask(A=@n18, B=@n11, Radius=@n11) {attenuationRadius=0.5, hardnessPercent=0}, Alpha=@n5.Blur) {constAlpha=0}) {constExponent=0.2}
  n20 = Multiply(A=@n6, B=ScalarParameter() {defaultValue=1, parameterName=RingPanSpeed})
  n21 = Multiply(A=@n19) {constB=2}
  n22 = Multiply(A=@n20) {constB=3.3}
  n23 = Add(
   A=TextureSample(UVs=Panner(Coordinate=@n19, Time=@n20) {speedY=1}) {texture=/Game/Effects/Fort_Effects/Textures/Grad/T_Halo_Corona_Gradient}.RGB
   B=TextureSample(UVs=Panner(Coordinate=@n21, Time=@n22) {speedY=1}) {texture=&T_Halo_Corona_Gradient}.RGB
  )
  MaterialFunctionCall(Emissive=Multiply(A=@n1.RGB, B=Multiply(A=@n17, B=@n23))) {materialFunction=&MF_FX_EmissiveAdjustments}
 ]
}
